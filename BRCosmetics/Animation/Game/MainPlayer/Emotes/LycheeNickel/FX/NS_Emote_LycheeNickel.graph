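# Niagara system for the LycheeNickel traversal emote, producing dust, debris, speed lines and wind streaks around the moving player.
# Two sprite dust layers and bouncing debris react to owner velocity through spawn speed and drag, with quality-dependent debris collision.
# Short-lived mesh speed lines and socket-attached, animated wind-streak sprites add motion accents.
# Front-end modifiers adjust spawn rates, alpha and camera offset, while time-of-day functions tint all but the debris.
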NiagaraSystem /BRCosmetics/Animation/Game/MainPlayer/Emotes/LycheeNickel/FX/NS_Emote_LycheeNickel
dependencies {
 usedDataInterfaces [
  &NiagaraDataInterfaceCurve
  &NiagaraDataInterfaceSpriteRendererInfo
  &NiagaraDataInterfaceVector2DCurve
  &NiagaraDataInterfaceArrayFloat
  &NiagaraDataInterfaceArrayColor
  &NiagaraDataInterfaceColorCurve
  &NiagaraDataInterfacePlatformSet
  &NiagaraDataInterfaceSocketReader
 ]
 usedDynamicInputs [
  &MakeFloatFromVector2d
  &Multiply_Vector2
  &SimulationPosition
 ]
 usedModules [
  &self:SystemSpawnScript.SystemScriptSource.SystemScriptGraph.NiagaraNodeAssignment_0.SetVariables_#1
  &SystemState
  &BlendSpaceHelper
 ]
}
emitters {
 Default_Emote_TraversalDust {
  emitterSpawnScript [/SetVariables(Emitter.FE Particle Spawn Rate=20, Emitter.Max Traversal Speed=448, Emitter.Particle Spawn Rate=20) {isSetParameters=true, module=SetVariables_#2}]
  emitterUpdateScript [
   &EmitterState()
   &SpawnRate(
    SpawnRate=&FN_FrontEndFloatModifier(
     Gameplay=&MakeCustomFloatFromBool(
      Bool=&SetBoolByFloatComparison(A=&VectorLength(Vector=$User.OwnerLocalVelocity), B=150)
      True Float=$Emitter.Particle Spawn Rate
      False Float=20
     )
     Front End=$Emitter.FE Particle Spawn Rate
    )
    Use Spawn Probability=false
   )
  ]
  particleSpawnScript [
   &InitializeParticle(
    Color Channel Mode:ENiagaraChannelCorrelation=::NewEnumerator0
    Color Maximum=<rgba 1 0.9915 0.9635 1>
    Color Minimum=<rgba 0.765 0.7502 0.7265 1>
    Color Mode=::NewEnumerator2
    Lifetime Max=0.26
    Lifetime Min=0.3
    Lifetime Mode=::NewEnumerator1
    Sprite Rotation Mode=::NewEnumerator2
    Sprite Size Mode=::NewEnumerator1
    Uniform Sprite Size Max=50
    Uniform Sprite Size Min=70
    UsePositionOffset=false
    Write Material Random=true
   ) {module=InitializeParticle001}
   &ShapeLocation(
    MaskSpawnGroup=false
    Offset=<0 0 -3>
    Offset Coordinate Space=::Local
    Offset Mode=::NewEnumerator0
    Override Seed=false
    Sphere Radius=10
    UseOwnerScale=true
   )
   &AddVelocity(
    Velocity=<0 -2 0.2>
    Velocity Speed Scale=&MakeCustomFloatFromBool(
     Bool=&SetBoolByFloatComparison(A=&VectorLength(Vector=$User.OwnerLocalVelocity), B=150)
     True Float=&RandomRangeFloat(Minimum=550, Maximum=750)
     False Float=&RandomRangeFloat(Minimum=800, Maximum=1000)
    )
   )
   &CameraOffset(Camera Offset Amount=15)
  ]
  particleUpdateScript [
   &ParticleState()
   &Drag(
    Drag=&FloatFromCurve(
     FloatCurve=DataInterface() {propertyValues {Curve {
        defaultValue=1
        keys [{interpMode=RCIM_Cubic, tangentMode=RCTM_User, time=0, value=0.75}, {interpMode=RCIM_Cubic, tangentMode=RCTM_User, time=0.65, value=1}]
       }}}
     Scale Curve=&Lerp_Float(
      A=20
      B=8
      Alpha=&RemapRange(
       Input Value To Remap=&VectorLength(Vector=$User.OwnerLocalVelocity)
       Input Min=0
       Input Max=$Emitter.Max Traversal Speed
       Output Min=0
       Output Max=1
      )
     )
    )
    Use Linear Drag=true
    Use Rotational Drag=true
   )
   &ScaleSpriteSize(Uniform Curve Sprite Scale=DataInterface() {propertyValues {Curve {keys [
        {arriveTangent=0.1005, interpMode=RCIM_Cubic, leaveTangent=0.1005, tangentMode=RCTM_User, time=0, value=1}
        {arriveTangent=0.3, interpMode=RCIM_Cubic, leaveTangent=0.3, time=0.4447, value=1.1526}
        {arriveTangent=0.0238, interpMode=RCIM_Cubic, leaveTangent=0.0238, tangentMode=RCTM_User, time=1, value=1.3}
       ]}}})
   &DynamicMaterialParameters(
    Index 0 Param 1=&Add_Float(
     A=&RandomRangeFloat(Minimum=0, Maximum=1)
     B=&FloatFromCurve(
      FloatCurve=DataInterface() {propertyValues {Curve {keys [
          <at 0 0 RCIM_Cubic>
          {arriveTangent=1.5947, interpMode=RCIM_Cubic, leaveTangent=1.5947, tangentMode=RCTM_User, time=0.2557, value=0.5257}
          <at 1 1 RCIM_Cubic>
         ]}}}
      CurveIndex=$Particles.Age
      Scale Curve=0.25
     )
    )
    Module.Index0Float0Write=true
    Module.Index0Float1Write=true
    Module.Index0Float2Write=true
    Module.Index0Float3Write=true
   ) {module=DynamicMaterialParameters001}
   &ScaleColor(
    Scale Alpha=&FloatFromCurve(
     FloatCurve=DataInterface() {propertyValues {Curve {keys [
         {arriveTangent=0.126, interpMode=RCIM_Cubic, leaveTangent=0.126, tangentMode=RCTM_User, time=0, value=0}
         {
          arriveTangent=1.4601
          arriveTangentWeight=0.238
          interpMode=RCIM_Cubic
          leaveTangent=1.4601
          leaveTangentWeight=0.5513
          tangentMode=RCTM_User
          time=0.198
          value=0.548
         }
         {arriveTangent=-0.5014, interpMode=RCIM_Cubic, leaveTangent=-0.5014, tangentMode=RCTM_User, time=0.5273, value=0.9572}
         {arriveTangent=-0.3188, interpMode=RCIM_Cubic, leaveTangent=-0.3188, tangentMode=RCTM_User, time=0.9, value=0}
        ]}}}
     Scale Curve=&Multiply_Float(
      A=&FN_FrontEndFloatModifier(Gameplay=1, Front End=0.3)
      B=&FNTimeOfDayScalar(Night Scale=0.6, Morning Scale=0.4, Evening Scale=0.7)
     )
    )
    Scale RGB=&MakeVectorFromLinearColorRGB(Color=&FNTimeOfDayColor(
      Night Color=&RandomRangeLinearColor(Minimum=<rgba 0.2115 0.2763 0.3652 1>, Maximum=<rgba 0.1968 0.2666 0.3791 1>)
      Morning Color=&RandomRangeLinearColor(Minimum=<rgba 0.6429 0.3899 0.35 0.372>, Maximum=<rgba 0.805 0.5284 0.3519 0.3101>)
      Day Color=&RandomRangeLinearColor(Minimum=<rgba 0.478 0.3116 0.2325 0.8>, Maximum=<rgba 0.6152 0.4038 0.2689 0.9>)
      Evening Color=&RandomRangeLinearColor(Minimum=<rgba 0.3929 0.2531 0.1877 0.8>, Maximum=<rgba 0.3652 0.2629 0.1788 0.9>)
     ))
    ScaleA=true
    ScaleRGB=true
   )
   &BlendSpace(Rotation Contribution=0)
   &SolveForcesAndVelocity(Clamp Velocity=false, Limit Acceleration=false)
  ]
  rendererClasses [&NiagaraSpriteRendererProperties]
  renderers [{rendererClass=&NiagaraSpriteRendererProperties}]
 }
 Default_Emote_TraversalDust001 {
  emitterSpawnScript [/SetVariables(Emitter.FE Particle Spawn Rate=40, Emitter.Max Traversal Speed=448, Emitter.Particle Spawn Rate=40) {isSetParameters=true, module=SetVariables_#2}]
  emitterUpdateScript [
   &EmitterState()
   &SpawnRate(
    SpawnRate=&FN_FrontEndFloatModifier(
     Gameplay=&MakeCustomFloatFromBool(
      Bool=&SetBoolByFloatComparison(A=&VectorLength(Vector=$User.OwnerLocalVelocity), B=150)
      True Float=$Emitter.Particle Spawn Rate
      False Float=40
     )
     Front End=$Emitter.FE Particle Spawn Rate
    )
    Use Spawn Probability=false
   )
  ]
  particleSpawnScript [
   &InitializeParticle(
    Color Channel Mode:ENiagaraChannelCorrelation=::NewEnumerator0
    Color Maximum=<rgba 1 0.9494 0.783 1>
    Color Minimum=<rgba 0.765 0.7095 0.6202 1>
    Color Mode=::NewEnumerator2
    Lifetime Max=0.2
    Lifetime Min=0.3
    Lifetime Mode=::NewEnumerator1
    Sprite Rotation Mode=::NewEnumerator2
    Sprite Size Mode=::NewEnumerator1
    Uniform Sprite Size Max=50
    Uniform Sprite Size Min=60
    UsePositionOffset=false
    Write Material Random=true
   ) {module=InitializeParticle001}
   &ShapeLocation(
    MaskSpawnGroup=false
    Offset=<0 50 5>
    Offset Coordinate Space=::Local
    Offset Mode=::NewEnumerator0
    Override Seed=false
    Sphere Radius=10
    UseOwnerScale=true
   )
   &AddVelocity(
    Velocity=<0 -2 0.2>
    Velocity Speed Scale=&MakeCustomFloatFromBool(
     Bool=&SetBoolByFloatComparison(A=&VectorLength(Vector=$User.OwnerLocalVelocity), B=150)
     True Float=&RandomRangeFloat(Minimum=750, Maximum=840)
     False Float=&RandomRangeFloat(Minimum=1600, Maximum=1200)
    )
   )
   &CameraOffset(Camera Offset Amount=15)
  ]
  particleUpdateScript [
   &ParticleState()
   &Drag(
    Drag=&FloatFromCurve(
     FloatCurve=DataInterface() {propertyValues {Curve {
        defaultValue=1
        keys [{interpMode=RCIM_Cubic, tangentMode=RCTM_User, time=0, value=0.75}, {interpMode=RCIM_Cubic, tangentMode=RCTM_User, time=0.65, value=1}]
       }}}
     Scale Curve=&Lerp_Float(
      A=20
      B=5
      Alpha=&RemapRange(
       Input Value To Remap=&VectorLength(Vector=$User.OwnerLocalVelocity)
       Input Min=0
       Input Max=$Emitter.Max Traversal Speed
       Output Min=0
       Output Max=1
      )
     )
    )
    Use Linear Drag=true
    Use Rotational Drag=true
   )
   &ScaleSpriteSize(Uniform Curve Sprite Scale=DataInterface() {propertyValues {Curve {keys [
        {arriveTangent=0.1096, interpMode=RCIM_Cubic, leaveTangent=0.1096, tangentMode=RCTM_User, time=0, value=1}
        {arriveTangent=0.3, interpMode=RCIM_Cubic, leaveTangent=0.3, time=0.4612, value=1.137}
        {arriveTangent=0.0238, interpMode=RCIM_Cubic, leaveTangent=0.0238, tangentMode=RCTM_User, time=1, value=1.3}
       ]}}})
   &DynamicMaterialParameters(
    Index 0 Param 1=&FloatFromCurve(FloatCurve=DataInterface() {propertyValues {Curve {keys [
         {arriveTangent=-0.026, interpMode=RCIM_Cubic, leaveTangent=-0.026, tangentMode=RCTM_User, time=0, value=0.2}
         {arriveTangent=0.9315, interpMode=RCIM_Cubic, leaveTangent=0.9315, time=0.6046, value=0.4325}
         {arriveTangent=-0.0974, interpMode=RCIM_Cubic, leaveTangent=-0.0974, tangentMode=RCTM_User, time=0.8588, value=1}
        ]}}})
    Index 0 Param 2=&SelectFloatFromArray(
     Array Sampling Mode=::NewEnumerator0
     Direct Array Index=?
     Float Selection Array=DataInterface() {propertyValues {FloatData [0,1]}}
     Random Seed=0
    )
    Module.Index0Float0Write=true
    Module.Index0Float1Write=true
    Module.Index0Float2Write=true
    Module.Index0Float3Write=true
   ) {module=DynamicMaterialParameters001}
   &ScaleColor(
    Scale Alpha=&FloatFromCurve(
     FloatCurve=DataInterface() {propertyValues {Curve {keys [
         {arriveTangent=0.126, interpMode=RCIM_Cubic, leaveTangent=0.126, tangentMode=RCTM_User, time=0, value=0}
         {
          arriveTangent=1.4601
          arriveTangentWeight=0.238
          interpMode=RCIM_Cubic
          leaveTangent=1.4601
          leaveTangentWeight=0.5513
          tangentMode=RCTM_User
          time=0.198
          value=0.548
         }
         {arriveTangent=-0.5014, interpMode=RCIM_Cubic, leaveTangent=-0.5014, tangentMode=RCTM_User, time=0.577, value=0.9572}
         {arriveTangent=-0.3188, interpMode=RCIM_Cubic, leaveTangent=-0.3188, tangentMode=RCTM_User, time=0.9, value=0}
        ]}}}
     Scale Curve=&FN_FrontEndFloatModifier(Gameplay=1, Front End=0.5)
    )
    Scale RGB=&MakeVectorFromLinearColorRGB(Color=&FNTimeOfDayColor(
      Night Color=&RandomRangeLinearColor(Minimum=<rgba 0.2433 0.38 0.7541 1>, Maximum=<rgba 0.1208 0.2828 0.6846 1>)
      Morning Color=&RandomRangeLinearColor(Minimum=<rgba 0.6623 0.57 0.6846 1>, Maximum=<rgba 0.5411 0.4361 0.5219 1>)
      Day Color=&RandomRangeLinearColor(Minimum=<rgba 0.6198 0.5155 0.5028 0.8>, Maximum=<rgba 1 0.82 0.8321 0.9>)
      Evening Color=&RandomRangeLinearColor(Minimum=<rgba 0.478 0.338 0.2816 1>, Maximum=<rgba 0.781 0.6131 0.5763 1>)
     ))
    ScaleA=true
    ScaleRGB=true
   )
   &BlendSpace(Rotation Contribution=0)
   &SolveForcesAndVelocity(Clamp Velocity=false, Limit Acceleration=false)
  ]
  rendererClasses [&NiagaraSpriteRendererProperties]
  renderers [{rendererClass=&NiagaraSpriteRendererProperties}]
 }
 Emote_TraversalDebris {
  emitterSpawnScript [/SetVariables(Emitter.Max Traversal Speed=448, Emitter.Particle Spawn Rate=5) {isSetParameters=true}]
  emitterUpdateScript [
   &EmitterState(
    Enable Distance Culling=true
    Life Cycle Mode=::NewEnumerator1
    Loop Behavior=::NewEnumerator1
    Loop Duration Mode=::NewEnumerator1
    MaxDistance=1150
    Module.bUseMaxDistance=true
    Module.bUseMinDistance=false
    Scalability Mode=::NewEnumerator1
    UseLoopDelay=false
   )
   &SpawnRate(
    Spawn Probability=0.4
    SpawnRate=&MakeCustomFloatFromBool(
     Bool=&SetBoolByFloatComparison(A=&VectorLength(Vector=$User.OwnerLocalVelocity), B=150)
     True Float=$Emitter.Particle Spawn Rate
     False Float=3
    )
    Use Spawn Probability=true
   )
  ]
  particleSpawnScript [
   &InitializeParticle(
    Color=&ColorFromCurve(
     ColorCurve=DataInterface() {propertyValues {
       AlphaCurve {keys [<at -0.0023 1 RCIM_Cubic>]}
       BlueCurve {keys [<at -0.0023 0.2122 RCIM_Cubic>,<at 0.1383 0.2122>,<at 0.2178 0.1859>,<at 0.4486 0.75>,<at 0.7126 0.1859>,<at 1.0234 0.8802>]}
       GreenCurve {keys [<at -0.0023 0.2325 RCIM_Cubic>,<at 0.1383 0.2325>,<at 0.2178 0.2005>,<at 0.4486 0.8456>,<at 0.7126 0.2122>,<at 1.0234 0.9474>]}
       RedCurve {keys [<at -0.0023 0.25 RCIM_Cubic>,<at 0.1383 0.25>,<at 0.2178 0.2188>,<at 0.4486 1>,<at 0.7126 0.2188>,<at 1.0234 1>]}
      }}
     CurveIndex=&UniformRangedFloat(Minimum=0, Maximum=1)
    )
    Lifetime=&UniformRangedFloat(Minimum=1.1, Maximum=1.2)
    Sprite Rotation Mode=::NewEnumerator2
    Sprite Size Mode=::NewEnumerator3
    Uniform Sprite Size=&UniformRangedFloat(Minimum=10, Maximum=7)
    UsePositionOffset=false
    Write Material Random=true
   ) {module=InitializeParticle001}
   &ShapeLocation(
    MaskSpawnGroup=false
    Offset=<0 -60 30>
    Offset Coordinate Space=::Local
    Offset Mode=::NewEnumerator0
    Override Seed=false
    Sphere Radius=0
    UseOwnerScale=true
   )
   &AddVelocity(
    Cone Axis=&MakeCustomVectorFromBool(
     Bool=&SetBoolByFloatComparison(A=&VectorLength(Vector=$User.OwnerLocalVelocity), B=0)
     True Vector=<0 -0.5 2>
     False Vector=<0 -1 0.8>
    )
    Module.UseConeFalloff=false
    Module.UseVelDistribution=false
    Override Seed=false
    Velocity Mode=::NewEnumerator2
    Velocity Speed=&MakeCustomFloatFromBool(
     Bool=&SetBoolByFloatComparison(A=&VectorLength(Vector=$User.OwnerLocalVelocity), B=150)
     True Float=&RandomRangeFloat(Minimum=150, Maximum=180)
     False Float=&RandomRangeFloat(Minimum=400, Maximum=350)
    )
   )
  ]
  particleUpdateScript [
   &ParticleState()
   &Drag(
    Drag=&MakeCustomFloatFromBool(
     Bool=&SetBoolByFloatComparison(A=&VectorLength(Vector=$User.OwnerLocalVelocity), B=0)
     True Float=&FloatFromCurve(
      FloatCurve=DataInterface() {propertyValues {Curve {
         defaultValue=0
         keys [
          {interpMode=RCIM_Cubic, tangentMode=RCTM_User, time=0, value=0}
          {arriveTangent=4.8882, interpMode=RCIM_Cubic, leaveTangent=4.8882, tangentMode=RCTM_User, time=0.5494, value=1}
         ]
        }}}
      Scale Curve=5
     )
     False Float=&FloatFromCurve(
      FloatCurve=DataInterface() {propertyValues {Curve {
         defaultValue=0
         keys [
          {interpMode=RCIM_Cubic, tangentMode=RCTM_User, time=0, value=0}
          {arriveTangent=4.8882, interpMode=RCIM_Cubic, leaveTangent=4.8882, tangentMode=RCTM_User, time=0.6377, value=1}
         ]
        }}}
      Scale Curve=5
     )
    )
    Use Linear Drag=true
    Use Rotational Drag=false
   )
   &GravityForce()
   &Collision(
    Age Colliding Particles=false
    Collision Enabled=&IsPlatformSetActive(New Platform Set=DataInterface() {propertyValues {Platforms {qualityLevelMask=-8}}})
    KillOnCollision=false
    Module.Control Rest State Via Surface Normal=false
    Module.EnableMaxCollisionCount=false
    Module.Randomize Collision Normal Vector=false
    Radius Calculation Type=::NewEnumerator1
   )
   &Collision(
    Age Colliding Particles=false
    Analytical Collision Plane Position 1:Vector3f=<0 0 -20>
    Analytical Collision Plane Space:ENiagaraCoordinateSpace=::Local
    CPU Collision Type=::NewEnumerator3
    Collision Enabled=&IsPlatformSetActive(New Platform Set=DataInterface() {propertyValues {Platforms {qualityLevelMask=-28}}})
    KillOnCollision=false
    Module.Control Rest State Via Surface Normal=false
    Module.EnableMaxCollisionCount=false
    Module.Randomize Collision Normal Vector=false
    Perform 2 Analytical Collision Queries=false
    Restitution=0
   ) {module=Collision001}
   &ScaleColor(Scale RGB=<1.2 1.2 1.2>, ScaleA=true, ScaleRGB=true)
   &ScaleSpriteSize(Scale Factor=&Vector2DFromFloat(Value=&FloatFromCurve(FloatCurve=DataInterface() {propertyValues {Curve {keys [
          <at 0.7 0.9701>
          {arriveTangent=-3.2149, interpMode=RCIM_Cubic, leaveTangent=-3.2149, tangentMode=RCTM_User, time=0.8778, value=0.9191}
          <at 1.0011 0.002>
         ]}}})))
   &SpriteRotationRate()
   &BlendSpace("Position Blend (World <-> Local)":NiagaraFloat=&MakeCustomFloatFromBool(
     Bool=&SetBoolByFloatComparison(A=&VectorLength(Vector=$Engine.Owner.Velocity), B=600)
     True Float=1
     False Float=&FloatFromCurve(FloatCurve=DataInterface() {propertyValues {Curve {
         defaultValue=1
         keys [
          {arriveTangent=-4.2489, interpMode=RCIM_Cubic, leaveTangent=-4.2489, tangentMode=RCTM_User, time=0, value=0.5657}
          {arriveTangent=0.2533, interpMode=RCIM_Cubic, leaveTangent=0.2533, tangentMode=RCTM_User, time=0.3439, value=0}
         ]
        }}})
     Translation Contribution:NiagaraFloat=1
     Rotation Contribution:NiagaraFloat=0
    ))
   &SolveForcesAndVelocity(Clamp Velocity=false, Limit Acceleration=false)
   &ScaleSpriteSizeBySpeed(
    Max Scale Factor=&Vector2DFromFloat(Value=1)
    Min Scale Factor=&Vector2DFromFloat(Value=0.25)
    Velocity Threshold=200
   )
  ]
  rendererClasses [&NiagaraSpriteRendererProperties]
  renderers [{rendererClass=&NiagaraSpriteRendererProperties}]
 }
 Speed_Lines002 {
  emitterSpawnScript [/SetVariables(Emitter.FE Particle Spawn Rate=20, Emitter.Max Traversal Speed=448, Emitter.Particle Spawn Rate=20) {isSetParameters=true}]
  emitterUpdateScript [
   &EmitterState(Life Cycle Mode=::NewEnumerator1, Loop Behavior=::NewEnumerator1, Loop Duration Mode=::NewEnumerator1, UseLoopDelay=false)
   &SpawnRate(
    SpawnRate=&FN_FrontEndFloatModifier(
     Gameplay=&MakeCustomFloatFromBool(
      Bool=&SetBoolByFloatComparison(A=&VectorLength(Vector=$User.OwnerLocalVelocity), B=150)
      True Float=$Emitter.Particle Spawn Rate
      False Float=20
     )
     Front End=$Emitter.FE Particle Spawn Rate
    )
    Use Spawn Probability=false
   )
  ]
  particleSpawnScript [
   &PartitionParticles()
   &InitializeParticle(
    Color=&SelectLinearColorFromArray(
     Array Sampling Mode=::NewEnumerator0
     Direct Array Index=?
     Color Selection Array=DataInterface() {propertyValues {ColorData [<rgba 0.0226 0.0226 0.0226 1>, <rgba 0.1291 0.0984 0.0861 1>, <rgba 0.0781 0.0781 0.0781 1>, <rgba 0.0226 0.0226 0.0226 1>]}}
     Random Seed=0
    )
    Lifetime Max=0.1
    Lifetime Min=0.2
    Lifetime Mode=::NewEnumerator1
    Mesh Scale=&Multiply_VectorByFloat(Vector=<0.8 1.5 0.4>, Float=1.5)
    Mesh Scale Mode=::NewEnumerator4
    Position Mode=::NewEnumerator1
    UsePositionOffset=false
    Write Material Random=true
   )
   &InitialMeshOrientation(Mesh Orientation Mode=::NewEnumerator2, Orientation Vector=<0 1 0>, Use Rotation Vector=false)
   &CameraOffset(Camera Offset Amount=&FN_FrontEndFloatModifier(Gameplay=0, Front End=-30))
   &SystemLocation(
    MaskSpawnGroup=false
    Offset=/Niagara/DynamicInputs/Helpers/MaskPosition(
     Mask Source:NiagaraInt32=1
     Mask Parameter:NiagaraInt32=$Particles.PartitionParticles.Partition
     Mask:NiagaraInt32=0
     Masked Position:NiagaraPosition=&RandomRangeVector(Minimum=<20 0 10>, Maximum=<-20 0 30>)
     Passthrough Position:NiagaraPosition=&RandomRangeVector(Minimum=<20 0 10>, Maximum=<-20 0 30>)
    )
   )
  ]
  particleUpdateScript [
   &ParticleState()
   &ScaleMeshSize(Scale Factor=&MakeVector(
     X=&FloatFromCurve(FloatCurve=DataInterface() {propertyValues {Curve {keys [<at 0 0.4>,<at 1 1>]}}})
     Y=&FloatFromCurve(FloatCurve=DataInterface() {propertyValues {Curve {defaultValue=1, keys [{tangentMode=RCTM_User, time=0, value=1}, {tangentMode=RCTM_User, time=1, value=0.4}]}}})
     Z=&FloatFromCurve(FloatCurve=DataInterface() {propertyValues {Curve {defaultValue=1, keys [{tangentMode=RCTM_User, time=0, value=1}, {tangentMode=RCTM_User, time=1, value=1}]}}})
    ))
   &Drag(Drag=5, Use Linear Drag=true, Use Rotational Drag=false)
   &SolveForcesAndVelocity(Clamp Velocity=false, Limit Acceleration=false)
   &ScaleColor(
    Scale Alpha=&FloatFromCurve(FloatCurve=DataInterface() {propertyValues {Curve {keys [
         {interpMode=RCIM_Cubic, tangentMode=RCTM_User, time=0.1552, value=0}
         {interpMode=RCIM_Cubic, tangentMode=RCTM_User, time=0.5, value=1}
         {tangentMode=RCTM_User, time=1, value=0}
        ]}}})
    Scale RGB=&MakeVectorFromLinearColorRGB(Color=&FNTimeOfDayColor(Night Color=<rgba 0.1176 0.3079 1 1>, Morning Color=<rgba 1 1 1 1>, Day Color=<rgba 1 1 1 1>, Evening Color=<rgba 1 1 1 1>))
    ScaleA=true
    ScaleRGB=true
   )
   &ScaleSpriteSize(
    Non-Uniform Curve Sprite Scale=DataInterface() {propertyValues {
      XCurve {keys [{tangentMode=RCTM_User, time=0, value=0.5}, {tangentMode=RCTM_User, time=1, value=1.5}]}
      YCurve {keys [{tangentMode=RCTM_User, time=0, value=1}, {tangentMode=RCTM_User, time=1, value=0.5}]}
     }}
    Scale Sprite Size Mode=::NewEnumerator3
   )
  ]
  rendererClasses [&NiagaraMeshRendererProperties]
  renderers [{rendererClass=&NiagaraMeshRendererProperties, rendererIndex=0}]
 }
 WindstreaksJet {
  emitterSpawnScript [/SetVariables(Emitter.FE Particle Spawn Rate=40, Emitter.Max Traversal Speed=448, Emitter.Particle Spawn Rate=40) {isSetParameters=true}]
  emitterUpdateScript [
   &EmitterState(Life Cycle Mode=::NewEnumerator1, Loop Behavior=::NewEnumerator1, Loop Duration Mode=::NewEnumerator1, UseLoopDelay=false)
   &SpawnRate(
    SpawnRate=&FN_FrontEndFloatModifier(
     Gameplay=&MakeCustomFloatFromBool(
      Bool=&SetBoolByFloatComparison(A=&VectorLength(Vector=$User.OwnerLocalVelocity), B=150)
      True Float=$Emitter.Particle Spawn Rate
      False Float=40
     )
     Front End=$Emitter.FE Particle Spawn Rate
    )
    Use Spawn Probability=false
   )
  ]
  particleSpawnScript [
   &PartitionParticles()
   &InitializeParticle(
    Color=<rgba 0.7633 0.628 0.5553 1>
    Lifetime=&UniformRangedFloat(Minimum=0.04, Maximum=0.08)
    Position Offset=<0 20 0>
    Sprite Size=&Multiply_Vector2D_ByFloat(Float=1, Vector2D=&UniformRangedVector2D(Minimum=<20 40>, Maximum=<30 50>))
    Sprite Size Mode=::NewEnumerator4
    UsePositionOffset=false
    Write Material Random=true
   )
   &SocketLocation(
    SamplePosOffset=true
    Sampled Position Offset=<0 -13 -17>
    Sampled Position Offset Coordinate Space:ENiagaraSkelMeshTransforms=::NewEnumerator2
    Socket Index=$Particles.PartitionParticles.Partition
    Socket Reader=$System.Socket Reader
    Socket Sampling Mode=::NewEnumerator3
    VelocitySpeedLimit=false
   )
   &AddVelocity(Velocity=&RandomRangeVector(Minimum=<50 -100 40>, Maximum=<-50 -200 20>), Velocity Speed Scale=0.1)
   &ApplyInitialForces()
   &SubUVAnimation(
    Evaluation Type=::NewEnumerator0
    "Sprite Renderer "=DataInterface() {propertyValues {SpriteRenderer=&self:WindstreaksJet_0.NiagaraSpriteRendererProperties_0}}
    SubUV Animation Mode:ENiagara_SubUVLookupModeV2=::NewEnumerator2
    UseEndFrame=false
    UseStartFrame=false
   )
  ]
  particleUpdateScript [
   &ParticleState()
   &ScaleSpriteSize(Scale Factor=&Multiply_Vector2D_ByFloat(
     Float=&FloatFromCurve(
      FloatCurve=DataInterface() {propertyValues {Curve {keys [
          {arriveTangent=10.1318, interpMode=RCIM_Cubic, leaveTangent=10.1318, tangentMode=RCTM_User, time=0, value=0}
          {interpMode=RCIM_Cubic, tangentMode=RCTM_User, time=0.2979, value=1}
         ]}}}
      CurveIndex=$Emitter.Age
     )
     Vector2D=&Vector2DFromCurve(
      Vector2Curve=DataInterface() {propertyValues {
        XCurve {keys [
          {arriveTangent=2.5428, interpMode=RCIM_Cubic, leaveTangent=2.5428, tangentMode=RCTM_User, time=0, value=1}
          {arriveTangent=0.1281, interpMode=RCIM_Cubic, leaveTangent=0.1281, tangentMode=RCTM_User, time=1, value=2}
         ]}
        YCurve {keys [
          {arriveTangent=2.8823, interpMode=RCIM_Cubic, leaveTangent=2.8823, tangentMode=RCTM_User, time=0, value=1}
          {arriveTangent=2.0297, interpMode=RCIM_Cubic, leaveTangent=2.0297, tangentMode=RCTM_User, time=1, value=2}
         ]}
       }}
      ScaleCurve=&Multiply_Vector2D_ByFloat(
       Float=&RandomRangeFloat(Minimum=1, Maximum=2)
       Vector2D=&RandomRangeVector2D(Minimum=<0.6 0.75>, Maximum=<0.8 1>)
      )
     )
    ))
   &DynamicMaterialParameters(
    Index 0 Param 1=5
    Index 0 Param 2=10
    Index 0 Param 3=1
    Index 0 Param 4=&FloatFromCurve(FloatCurve=DataInterface() {propertyValues {Curve {keys [{tangentMode=RCTM_User, time=0.1, value=0}, {tangentMode=RCTM_User, time=1, value=0}]}}})
    Module.Index0Float0Write=true
    Module.Index0Float1Write=true
    Module.Index0Float2Write=true
    Module.Index0Float3Write=true
   ) {module=DynamicMaterialParameters001}
   &Drag(Drag=5, Use Linear Drag=true, Use Rotational Drag=true)
   &SolveForcesAndVelocity(Clamp Velocity=false, Limit Acceleration=false)
   &ScaleColor(
    Scale Mode=::NewEnumerator1
    Scale RGBA=&MakeVector4_FromLinearColor(LinearColor=&FNTimeOfDayColor(
      Night Color=&RandomRangeLinearColor(Minimum=<rgba 0.1408 0.2438 0.467 0.4879>, Maximum=<rgba 0.2445 0.3598 0.6429 0.3256>)
      Morning Color=&RandomRangeLinearColor(Minimum=<rgba 0.478 0.343 0.4223 0.7>, Maximum=<rgba 0.5087 0.3764 0.4283 0.4184>)
      Day Color=&RandomRangeLinearColor(Minimum=<rgba 0.478 0.454 0.4473 0.7>, Maximum=<rgba 0.781 0.7524 0.7417 1>)
      Evening Color=&RandomRangeLinearColor(Minimum=<rgba 0.3244 0.3837 0.3766 0.2947>, Maximum=<rgba 0.3383 0.3937 0.4161 0.1633>)
     ))
    ScaleRGBA=true
   )
   &ScaleColor(
    Scale Alpha=&FloatFromCurve(FloatCurve=DataInterface() {propertyValues {Curve {keys [<at 0 0>,<at 0.0483 0.9765>,<at 0.8887 0.9107>,<at 1 0>]}}}, Scale Curve=1.5)
    ScaleA=true
    ScaleRGB=false
   ) {module=ScaleColor002}
  ]
  rendererClasses [&NiagaraSpriteRendererProperties]
  renderers [{rendererClass=&NiagaraSpriteRendererProperties}]
 }
}
summary {
 systemName=NS_Emote_LycheeNickel
 emitters [
  {emitterName=Default_Emote_TraversalDust, rendererClasses [&NiagaraSpriteRendererProperties]}
  {emitterName=WindstreaksJet, rendererClasses [&NiagaraSpriteRendererProperties]}
  {emitterName=Default_Emote_TraversalDust001, rendererClasses [&NiagaraSpriteRendererProperties]}
  {emitterName=Speed_Lines002, rendererClasses [&NiagaraMeshRendererProperties]}
  {emitterName=Emote_TraversalDebris, rendererClasses [&NiagaraSpriteRendererProperties]}
 ]
 userVariables [
  User.bIsFrontEnd {type=NiagaraBool}
  User.bIsFrontEndPreview {type=NiagaraBool}
  User.OwnerLocalVelocity {type=Vector3f, description="", defaultValue {struct=Vector3f, value=<0 0 0>}}
  User.WorkoutSelect {type=NiagaraFloat, defaultValue=0, description=""}
 ]
}
